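# Niagara trail effect for the male Backspin glider, with red ribbons and glow lights attached to its skeletal mesh.
# Streaks_Red spawns two ribbons once fully deployed, stretching them along the sampled mesh axis by age, Ribbon Length and a system integrator value.
# GlowLight keeps two socket-following particles drawing sprites and lights.
# ForwardDot drives streak brightness plus glow color, alpha, light radius and second-sprite size.
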NiagaraSystem /BRCosmetics/Gadgets/Assets/VinderTech_GliderChute/Glider_BackspinMale/FX/NS_Glider_Backspin_Male_Trail
dependencies {
 usedDataInterfaces [&NiagaraDataInterfaceCurve]
 usedDynamicInputs []
 usedModules [
  &SystemState
  &self:SystemSpawnScript.SystemScriptSource.SystemScriptGraph.NiagaraNodeAssignment_1.SetVariables_#1
  /Game/Effects/Niagara/Modules/Math/FN_Integrator/FN_Integrator_v2
 ]
}
emitters {
 GlowLight {
  emitterUpdateScript [
   &EmitterState(Life Cycle Mode=::NewEnumerator1, Loop Behavior=::NewEnumerator1, Loop Delay=0.1, Loop Duration Mode=::NewEnumerator1, UseLoopDelay=true)
   &SpawnBurst_Instantaneous(Spawn Count=2, Use Spawn Probability=false)
  ]
  particleSpawnScript [
   &FN_SetSpawnGroup_v2(Number of Spawn Groups=2)
   /Niagara/Modules/Spawn/Initialization/InitializeParticle(
    Color=<rgba 20 0.6 0 0.5>
    Lifetime=5
    Module.Write Color=true
    Module.Write Lifetime=false
    Module.Write Mass=true
    Module.Write Material Random=true
    Module.Write Position=false
    Module.Write Scale=false
    Module.Write SpriteRotation=false
    Module.Write SpriteSize=true
    Position=&SimulationPosition()
    Sprite Size=<18 18>
   )
   &DynamicMaterialParameters(
    Index 0 Param 1=50
    Module.Param0WriteEnabled=true
    Module.Param1WriteEnabled=true
    Module.Param2WriteEnabled=true
    Module.Param3WriteEnabled=true
   )
  ]
  particleUpdateScript [
   &ParticleState(Kill Particles When Lifetime Has Elapsed=false, Let Infinitely Lived Particles Die When Emitter Deactivates=true)
   &SkeletalMeshLocation(
    Bone / Socket Index=$Particles.SpawnGroup
    Inherit Velocity=false
    Interpolate Mesh Sampling=false
    MaskSpawnGroup=false
    Mesh Sampling Type=::NewEnumerator5
    Sampled Position Offset=&Lerp_Vector(A=<10 -1 0>, B=<10 1 0>, Alpha=&MakeFloatFromInt(Int=$Particles.SpawnGroup))
    Sampled Position Offset Coordinate Space:ENiagaraMeshTransforms=::NewEnumerator3
    Skeletal Mesh=$User.SkeletalMesh
    Socket Sampling Mode=::NewEnumerator5
   )
   &FN_Light_v3(
    Color=<rgba 1 0.0398 0 1>
    Create Color Attribute=true
    Create Exponent Attribute=true
    Create Position Attribute=true
    Create Radius Attribute=true
    Enable Particle Light Enable=true
    Exponent=5
    Light Enabled=true
    Position:Vector3f=&Add_Vector(
     A=$Output.SkeletalMeshLocation.SampledPosition
     B=&RotateVectorByQuaternion(Vector To Rotate=<15 0 0>, Quaternion=$Output.SkeletalMeshLocation.SampledRotation)
    )
    Radius:NiagaraFloat=&Lerp_Float(A=40, B=100, Alpha=$User.ForwardDot)
    Scale Color=20
   )
   &ScaleSpriteSize(Scale Factor=&Vector2DFromFloat(Value=&Lerp_Float(A=0.7, B=1, Alpha=$User.ForwardDot))) {enabled=false}
   /SetVariables(Particles.Renderer2_Size:Vector2f=&Vector2DFromFloat(Value=&Lerp_Float(A=0, B=75, Alpha=$User.ForwardDot))) {isSetParameters=true}
   &ScaleColor(
    Scale Alpha=&Lerp_Float(A=0.1, B=0.3, Alpha=$User.ForwardDot)
    Scale RGB=&VectorFromFloat(Value=&Lerp_Float(A=1, B=2, Alpha=$User.ForwardDot))
    ScaleA=true
    ScaleRGB=true
   )
   /SetVariables(Particles.Renderer2_Color:LinearColor=&MultiplyLinearColor(Linear Color=$Particles.Initial.Color, Scale Factor=<1 1 1 0.015>)) {isSetParameters=true}
   &CameraOffset(Camera Offset Amount=5)
  ]
  rendererClasses [&NiagaraSpriteRendererProperties, &NiagaraLightRendererProperties]
  renderers [
   {rendererClass=&NiagaraSpriteRendererProperties}
   {rendererClass=&NiagaraLightRendererProperties, rendererIndex=1}
   {rendererClass=&NiagaraSpriteRendererProperties, rendererIndex=2}
  ]
 }
 Streaks_Red {
  emitterUpdateScript [
   &EmitterState()
   &FN_SpawnRibbon_v2(
    Number of Ribbons=2
    SpawnRate:NiagaraFloat=&Lerp_Float(A=0, B=20, Alpha=&MakeCustomFloatFromBool(Bool=$User.bIsFullyDeployed, True Float=1, False Float=0))
   )
  ]
  particleSpawnScript [
   &FN_InitializeRibbon_v3(
    Color:LinearColor=<rgba 1 0.0124 0 1>
    Enable Skeletal Position=true
    Facing Axis:ENiagaraOrientationAxis=::ZAxis
    Interpolate Bone Sampling=false
    Lifetime=1.5
    Ribbon Width=23
   )
   /SetVariables(Particles.RibbonFacing=&RotateVectorByQuaternion(Vector To Rotate=<0 0 1>, Quaternion=$Particles.SampledRotation)) {isSetParameters=true}
  ]
  particleUpdateScript [
   &ParticleState()
   /Game/Effects/Niagara/Modules/Contrails/FN_UpdateRibbon/FN_UpdateRibbon_v3(Enable Skeletal Position=true)
   /SetVariables(Particles.Position=&Add_Vector(
     A=Linked() {linkedVariable {name=Particles.Initial.Position, type=NiagaraPosition}}
     B=&Multiply_VectorByFloat(
      Vector=&RotateVectorByQuaternion(
       Vector To Rotate=&MakeVector(
        X=&Multiply_Float(A=&Lerp_Float(A=150, B=125, Alpha=$System.FN_Integrator_v2.Attr), B=$User.Ribbon Length)
        Y=0
        Z=0
       )
       Quaternion=$Particles.SampledRotation
      )
      Float=$Particles.NormalizedAge
     )
    )) {isSetParameters=true}
   /SetVariables(Particles.RibbonFacing=&MakeCustomVectorFromBool(
     Bool=$Particles.Locked
     True Vector=&RotateVectorByQuaternion(Vector To Rotate=<0 0 1>, Quaternion=$Particles.FN_UpdateRibbon_v3.SampledRotation)
     False Vector=$Particles.RibbonFacing
    )) {isSetParameters=true}
   &ScaleColor(
    Scale Alpha=&FloatFromCurve(FloatCurve=DataInterface() {propertyValues {Curve {keys [
         {arriveTangent=12.3493, interpMode=RCIM_Cubic, leaveTangent=12.3493, tangentMode=RCTM_User, time=0, value=0}
         <at 0.0966 1>
         <at 1 0 RCIM_Cubic>
        ]}}})
    Scale RGB=&VectorFromFloat(Value=&Lerp_Float(A=1, B=5, Alpha=$User.ForwardDot))
    ScaleA=true
    ScaleRGB=true
   )
   &ScaleRibbonWidth(Ribbon Width Scale=&FloatFromCurve(
     FloatCurve=DataInterface() {propertyValues {Curve {keys [
         {arriveTangent=5.2789, interpMode=RCIM_Cubic, leaveTangent=5.2789, tangentMode=RCTM_User, time=-0.0034, value=0.3}
         <at 0.15 1>
         <at 1 0 RCIM_Cubic>
        ]}}}
     Scale Curve=0.9
    ))
  ]
  rendererClasses [&NiagaraRibbonRendererProperties]
  renderers [{rendererClass=&NiagaraRibbonRendererProperties}]
 }
}
summary {
 systemName=NS_Glider_Backspin_Male_Trail
 emitters [
  {emitterName=Streaks_Red, rendererClasses [&NiagaraRibbonRendererProperties]}
  {emitterName=GlowLight, rendererClasses [&NiagaraSpriteRendererProperties, &NiagaraLightRendererProperties]}
 ]
 userVariables [
  User.bIsFrontEndPreview {type=NiagaraBool}
  User.bIsFullyDeployed {type=NiagaraBool}
  User.ForwardDot {type=NiagaraFloat, defaultValue=0, description=""}
  User.Jet Color {type=LinearColor, description="", defaultValue {struct=LinearColor, value=<rgba 1 1 1 1>}}
  User.Jet_Alpha {type=NiagaraFloat, defaultValue=1, description=""}
  User.PlayerSpeed {type=NiagaraFloat, defaultValue=0, description=""}
  User.Ribbon Alpha {type=NiagaraFloat, defaultValue=1, description=""}
  User.Ribbon Color {type=LinearColor, description="", defaultValue {struct=LinearColor, value=<rgba 1 1 1 1>}}
  User.Ribbon Length {type=NiagaraFloat, defaultValue=1, description=""}
  User.Ribbon Width {type=NiagaraFloat, defaultValue=50, description=""}
  User.RibbonIDOffset {type=NiagaraInt32}
  User.RightDot {type=NiagaraFloat, defaultValue=0, description=""}
  User.SkeletalMesh {
   type=NiagaraDataInterfaceSkeletalMesh
   defaultValue {
    struct=NiagaraExt_VariableValue_DataInterface
    value {dataInterface=&self:NiagaraDataInterfaceSkeletalMesh_0}
   }
  }
 ]
}
